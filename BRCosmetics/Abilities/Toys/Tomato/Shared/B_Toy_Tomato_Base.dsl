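; Base Blueprint for the thrown tomato toy, a spinning physics ball that bursts on hitting another player, the world or water.
; The server delays collision briefly, measures throw distance on impact, fires a gameplay cue and silently destroys the ball.
; Clients scale the ball up via timeline.
; Player hits set a replicated message whose OnRep prints thrower, victim and distance to the squad feed, and sufficiently long throws can award quest progress.
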(event EventBeginPlay
  (bind _spherecollision (Variables|FortPhysicsBall|GetSphereCollision))
  (Networking|SwitchHasAuthority
    (:Authority
      ;[1 0.3126 0] On Server
      (Collision|SetCollisionEnabled _spherecollision "NoCollision")
      (Utilities|FlowControl|Delay 0.25)
      (Collision|SetCollisionEnabled _spherecollision "QueryOnly")
      (if (Variables|Default|GetHasFinishedSpawning)
        (Variables|/Script/Engine.RotatingMovementComponent|SetRotationRate "-180.000000,360.000000,200.000000" (Variables|Default|GetRotatingMovement))))
    (:Remote
      (Audio|Components|Audio|SetVolumeMultiplier (Variables|Default|GetAudio) 1.0 :!class /Script/Engine.AudioComponent)
      ;[0 1 0.2211] Client
      (bind (_direction _scale_track) (Timeline|Timeline_0 PlayFromStart)
        (:Update
          (Transformation|SetWorldScale3D (Variables|Default|GetBall) (Math|Conversions|ToVector(Float) (Math|Float|Lerp (Variables|Default|GetMin%20Scale) (Variables|Default|GetMax%20Scale) _scale_track))))))))

(event Visuals|EventInitializeToyInstance
  (params (OwningPC :type Fort%20Player%20Controller%20Object%20Reference) (NumTimesSummoned :type Integer))
  (|CallParentFunction OwningPC NumTimesSummoned)
  (Variables|Default|SetHasFinishedSpawning true))

(event EventProjectileStop
  (params (ImpactResult :type Hit%20Result%20Structure%20%28by%20ref%29))
  (|CallParentFunction ImpactResult)
  (Variables|Default|SetHasComeToRest true)
  (Audio|Components|Audio|FadeOut (Variables|Default|GetAudio) 0.0 0.0 "Linear" :!class /Script/Engine.AudioComponent))

(event EventOnPawnTouchedMe
  (params (TouchingPawn :type Pawn%20Object%20Reference) (ImpactNormal :type Vector :default "0, 0, 0") (OtherVelocity :type Vector :default "0, 0, 0"))
  (bind _returnvalue (Actor|GetOwner))
  (bind _returnvalue_2 (Fortnite|GetFortPlayerControllerfromActor _returnvalue))
  ;[0.2656 0.08305 0] (On Server) Explode the tomato if it hits another player
  (|CallParentFunction TouchingPawn ImpactNormal OtherVelocity)
  (bind _asfort_player_pawn (Utilities|Casting|CastToFortPlayerPawn TouchingPawn)
    (:then
      (bind _asfort_player_state (Utilities|Casting|CastToFortPlayerState (Variables|/Script/Engine.Controller|GetPlayerState (Pawn|GetController _asfort_player_pawn)))))
    (:CastFailed
      (Utilities|Casting|CastToFortAIPawn TouchingPawn)))
  (if (Math|Boolean|ANDBoolean (Utilities|NotEqual(Object) _asfort_player_state (Actor|GetOwner)) (not (Variables|Default|GetHasComeToRest)) (Variables|Default|GetHasFinishedSpawning))
    (Networking|SwitchHasAuthority
      (:Authority
        (Variables|/Script/Engine.RotatingMovementComponent|SetRotationRate "0, 0, 0" (Variables|Default|GetRotatingMovement))
        (bind _output_get (Variables|Default|SetDistanceTraveled (Math|Vector|VectorLength (- (Variables|Default|GetInitialSpawnPosition) (Transformation|GetActorLocation)))))
        (CallFunction|PlayCueCenteredonMe (Utilities|Array|Get(acopy) (Variables|Default|GetCuesForScoreValues)) _output_get)
        (Utilities|FlowControl|Sequence
          (:then_0
            (Utilities|IsValid _asfort_player_state
              (:Is%20Valid
                (bind _asfort_player_state_2 (Utilities|Casting|CastToFortPlayerState (Actor|GetOwner)))
                (CallFunction|CreateMessage _asfort_player_state_2 _asfort_player_state)
                (if (>= (Math|Conversions|ToFloat(Integer) (Math|Float|Truncate (/ (Variables|Default|GetDistanceTraveled) 1e+02))) (Variables|Default|GetDistanceForQuestReward))
                  (bind _cancompletequest (Class|AthenaQuestFunctionLibrary|AthenaCheckCreativeMode (Game|GetGameState)))
                  (if _cancompletequest
                    (Variables|Default|SetControllerWhoThrewTomato _returnvalue_2)
                    (bind (_questisvalid _questcompleted) (Default|AthenaCheckQuestandObjectiveCompletion _returnvalue_2 (Variables|Default|GetQuestItem) (Variables|Default|GetObjBackendName)))
                    (if (and _questisvalid (not _questcompleted))
                      (Class|AthenaQuestFunctionLibrary|AthenaGiveQuestUpdatetoPlayers :InQuestItemRef (Variables|Default|GetQuestItem) :ObjectiveBackendName (Variables|Default|GetObjBackendName) :PlayerControllersForUpdate (Utilities|Array|MakeArray _returnvalue_2))
                      (CallFunction|PlayQuestSuccessFX)))))))
          (:then_1
            (Gameplay|SilentDie false))))
      (:Remote
        ;[0 1 0.2211] On Client
        (Development|SetHiddeninGame (Variables|Default|GetBall) true false)))))

(event Custom|Play%20Quest%20Success%20FX
  (bind _controllerwhothrewtomato (Variables|Default|GetControllerWhoThrewTomato))
  (Networking|SwitchHasAuthority
    (:Remote
      (Utilities|IsValid _controllerwhothrewtomato
        (:Is%20Valid
          (bind _asfort_player_controller_athena (Utilities|Casting|CastToFortPlayerControllerAthena _controllerwhothrewtomato))
          (Audio|PlaySoundatLocation (Variables|Default|GetQuestSuccessSound) (Transformation|GetActorLocation (Game|Player|GetPlayerPawn _asfort_player_controller_athena)))
          (Effects|SpawnFXSystemAttached 0 (Variables|/Script/Engine.Character|GetMesh (Game|Player|GetPlayerPawn _asfort_player_controller_athena))))))))

(event Collision|EventHit
  (params (MyComp :type Primitive%20Component%20Object%20Reference) (Other :type Actor%20Object%20Reference) (OtherComp :type Primitive%20Component%20Object%20Reference) (bSelfMoved :type Boolean) (HitLocation :type Vector :default "0, 0, 0") (HitNormal :type Vector :default "0, 0, 0") (NormalImpulse :type Vector :default "0, 0, 0") (Hit :type Hit%20Result%20Structure%20%28by%20ref%29))
  (bind (_bblockinghit _binitialoverlap _time _distance _location _impactpoint _normal _impactnormal _physmat _hitactor _hitcomponent) (Collision|BreakHitResult Hit))
  ;[0 0 0] Explode the tomato if it hits the world
  (|CallParentFunction MyComp Other OtherComp bSelfMoved HitLocation HitNormal NormalImpulse Hit)
  (if _bblockinghit
    (if (or (Utilities|Enum|Equal(Enum) (Collision|GetCollisionObjectType _hitcomponent) "ECC_WorldStatic") (Utilities|Enum|Equal(Enum) (Collision|GetCollisionObjectType _hitcomponent) "ECC_WorldDynamic"))
      (Development|SetHiddeninGame (Variables|Default|GetBall) true false)
      (Variables|/Script/Engine.RotatingMovementComponent|SetRotationRate "0, 0, 0" (Variables|Default|GetRotatingMovement))
      (Variables|Default|SetDistanceTraveled (Math|Vector|VectorLength (- (Variables|Default|GetInitialSpawnPosition) _location)))
      (bind _outcueparameters (GameplayCue|BuildCueParametersfromHitResult self Hit))
      (Ability|ExecuteGameplayCueWithParamsReplicated (Utilities|Array|Get(acopy) (Variables|Default|GetCuesForScoreValues)) (Ability|GameplayCue|MakeGameplayCueParameters :NormalizedMagnitude (Ability|GameplayCue|BreakGameplayCueParameters _outcueparameters) :RawMagnitude (.RawMagnitude (Ability|GameplayCue|BreakGameplayCueParameters _outcueparameters)) :EffectContext "()" :AggregatedSourceTags "(GameplayTags=)" :AggregatedTargetTags "(GameplayTags=)" :Location (.Location (Ability|GameplayCue|BreakGameplayCueParameters _outcueparameters)) :Normal (.Normal (Ability|GameplayCue|BreakGameplayCueParameters _outcueparameters)) :Instigator (.Instigator (Ability|GameplayCue|BreakGameplayCueParameters _outcueparameters)) :GameplayEffectLevel 1 :AbilityLevel 1 :bReplicateLocationWhenUsingMinimalRepProxy false))
      (Gameplay|SilentDie false))))

(event OnEnterWaterEvent(WaterInteractionComponent)
  (params (WaterBody :type Fort%20Water%20Body%20Actor%20Object%20Reference) (WaterInteractionComponent :type Fort%20Water%20Interaction%20Component%20Object%20Reference) (bIsFirstBody :type Boolean))
  ;[0 0 0] Explode the tomato if it hits the world
  (Ability|ExecuteGameplayCueWithParamsReplicated (Utilities|Array|Get(acopy) (Variables|Default|GetCuesForScoreValues)) (Ability|GameplayCue|MakeGameplayCueParameters :NormalizedMagnitude 0.0 :RawMagnitude 0.0 :EffectContext "()" :AggregatedSourceTags "(GameplayTags=)" :AggregatedTargetTags "(GameplayTags=)" :Location "0,0,0" :Normal "0,0,0" :GameplayEffectLevel 1 :AbilityLevel 1 :bReplicateLocationWhenUsingMinimalRepProxy false))
  (Gameplay|SilentDie false))

(fn CreateMessage
  (params (Display%20Message%20Ball%20Owner :type Fort%20Player%20State%20Object%20Reference) (OtherPlayer :type Fort%20Player%20State%20Object%20Reference))
  (if (Variables|/Script/FortniteGame.FortRuntimeOptions|GetbToyMessagingEnabled (Options|GetRuntimeOptions))
    (Variables|Default|SetMessageContents (Utilities|Struct|MakeStructTomatoMinigameMessage Display%20Message%20Ball%20Owner (Math|Conversions|ToFloat(Integer) (Math|Float|Truncate (/ (Variables|Default|GetDistanceTraveled) 1e+02))) OtherPlayer))))

(fn GetDistanceText
  (params (Distance :type Float%20%28double-precision%29))
  (if (> Distance 5.0)
    (return (Utilities|Text|ToText(Integer) (Math|Float|Round Distance)))
    (else
      (return (Utilities|Text|ToText(Float) Distance "HalfToEven" false true 1 324 1 1)))))

(fn OnRep_MessageContents
  (bind _messagecontents (Variables|Default|GetMessageContents))
  (bind (_ballowner_22_ef5d780b4020e911ab429d9ba6b977ba _distance_33_9facd4014a6f7fdaaa915396bd79b184 _otherplayer_40_55ba47c849cd05ac123c95bb31600857) (Utilities|Struct|BreakStructTomatoMinigameMessage _messagecontents))
  (Networking|SwitchHasAuthority
    (:Remote
      (Utilities|IsValid _ballowner_22_ef5d780b4020e911ab429d9ba6b977ba
        (:Is%20Valid
          (Utilities|IsValid _otherplayer_40_55ba47c849cd05ac123c95bb31600857
            (:Is%20Valid
              (Utilities|IsValid (BlueprintContext|GetFortToyContext)
                (:Is%20Valid
                  (bind _returnvalue (Class|FortToyContext|FormatPlayerNameforDisplay (BlueprintContext|GetFortToyContext) _ballowner_22_ef5d780b4020e911ab429d9ba6b977ba "ToyOwner"))
                  (bind _returnvalue_2 (Class|FortToyContext|FormatPlayerNameforDisplay (BlueprintContext|GetFortToyContext) _otherplayer_40_55ba47c849cd05ac123c95bb31600857 "OtherPlayerInteractingWithToy"))
                  (Class|FortToyContext|ConditionalPrinttoMessageFeed (BlueprintContext|GetFortToyContext) _ballowner_22_ef5d780b4020e911ab429d9ba6b977ba (Transformation|GetActorLocation) "SquadMembersOnly" (Utilities|Text|FormatText (Variables|TomatoConfig|GetHitFormatText) _returnvalue _returnvalue_2 (Variables|TomatoConfig|GetToyDisplayName) (CallFunction|GetDistanceText _distance_33_9facd4014a6f7fdaaa915396bd79b184))))))))))))

(fn ConstructionScript
  (|CallParentFunction))
